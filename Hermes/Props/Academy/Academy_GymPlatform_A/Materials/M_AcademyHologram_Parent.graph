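# Unlit masked parent material for Academy hologram props such as the gym platform, with animated cells and a spawn in/out dissolve.
# Hashed, object-stabilized screen-space grids create cells and UV distortion; curvature and Fresnel index a color curve atlas.
# Custom primitive data SpawnInOut drives a noisy directional gradient for opacity and an edge mask.
# Static switches select vertex offset: coin bob/spin with flicker-timed noise glitches and height pulse, masked pull toward actor origin, or none.
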Material /Hermes/Props/Academy/Academy_GymPlatform_A/Materials/M_AcademyHologram_Parent
graph {
 properties {bAutomaticallySetUsageInEditor=false, blendMode=BLEND_Masked, shadingModel=MSM_Unlit, usedShadingModels=Unlit}
 nodes [
  n1 = LinearInterpolate(
   A=ScalarParameter() {defaultValue=0.25, parameterName=Cell_ColorMinInfluence}
   B=ScalarParameter() {defaultValue=0.35, parameterName=Cell_ColorMaxInfluence}
   Alpha=NamedRerouteUsage()
  ) {constA=0.25, constB=0.5}
  n2 = NamedRerouteUsage()
  n3 = LinearInterpolate(
   A=@n1
   B=ScalarParameter() {defaultValue=1, parameterName=ColorSampleMax}
   Alpha=ChannelMaskParameter(None=TextureSampleParameter2D(UVs=@n2) {parameterName=Curvature, texture=/Hermes/Props/Academy/Academy_GymPlatform_A/Textures/T_Academy_GymPlatform_A_A_M}.RGB) {maskChannel=Green, parameterName=CurvatureChannel, defaultValue=<rgba 0 1 0 0>}.R
  )
  Add(A=@n1, B=@n3)
  n4 = Time()
  n5 = NamedRerouteDeclaration(None=ScalarParameter() {bUseCustomPrimitiveData=true, parameterName=SpawnInOut, primitiveDataIndex=1}) {name=Spawn-In-Out, nodeColor=<rgba 0 1 0.8594 1>}
  ConstantBiasScale(None=Sine(None=Multiply(A=@n4, B=@n5)))
  Constant()
  EyeAdaptation()
  Multiply(
   A=NamedRerouteUsage()
   B=EyeAdaptationInverse(LightValueInput=ScalarParameter() {defaultValue=2, parameterName=Emissive_Strength}).EyeAdaptationInverse
  )
  n6 = ObjectPositionWS()
  n7 = ScalarParameter() {defaultValue=0.25, parameterName=ScreenUVDistortionMaskScale}
  n8 = MaterialFunctionCall(
   Seed=Time()
   Vec2=Ceil(None=MaterialFunctionCall!(Position=@n6, UVRadialScale=@n7) {materialFunction=/Game/Weapons/WeaponSkins/MaterialFunction/ObjectStabilizedScreenSpaceUVs}.StabilizedUVs)
  ) {materialFunction=&MF_Hash}
  n9 = Multiply(
   A=ComponentMask(None=Saturate(None=@n8.Vec2)) {r=true}
   B=Multiply(A=NamedRerouteUsage(), B=ScalarParameter() {defaultValue=0.1, parameterName=UVDistortStrength})
  )
  n10 = NamedRerouteUsage()
  NamedRerouteDeclaration(None=Add(A=TextureCoordinate(), B=Multiply(A=@n9, B=@n10))) {name=Base_TextureCoordinate, nodeColor=<rgba 0.7109 0 1 1>}
  n11 = ScalarParameter() {defaultValue=0.25, parameterName=Cell_Min_Value}
  n12 = ScalarParameter() {defaultValue=1, parameterName=Cell_Max_Value}
  n13 = MaterialFunctionCall(Position=ObjectPositionWS(), UVRadialScale=ScalarParameter() {defaultValue=0.01, parameterName=ScreenUVScale}) {materialFunction=&ObjectStabilizedScreenSpaceUVs}
  n14 = MaterialFunctionCall(Seed=Time(), Vec2=Ceil(None=@n13.StabilizedUVs)) {materialFunction=&MF_Hash}
  n15 = NamedRerouteDeclaration(None=Saturate(None=ComponentMask(None=@n14.Vec2) {r=true})) {name=AnimatedCells, nodeColor=<rgba 0 0.8594 1 1>}
  n16 = MaterialFunctionCall(Float2=Frac(None=@n13.StabilizedUVs)) {materialFunction=&BreakOutFloat2Components}
  n17 = Constant()
  n18 = ScalarParameter() {defaultValue=0.95, parameterName=CellWidth}
  n19 = ScalarParameter() {defaultValue=0.9, parameterName=CellHardness}
  n20 = Multiply(A=SphereMask(A=@n16.R, B=@n17, Radius=@n18, Hardness=@n19), B=SphereMask(A=@n16.G, B=@n17, Radius=@n18, Hardness=@n19))
  NamedRerouteDeclaration(None=Max(A=LinearInterpolate(A=@n11, B=@n12, Alpha=@n15), B=OneMinus(None=@n20))) {name=CellShape, nodeColor=<rgba 1 0.4453 0 1>}
  n21 = ScalarParameter() {defaultValue=0.5, parameterName=FresnelDarkValue}
  n22 = ScalarParameter() {defaultValue=1, parameterName=Exponent}
  n23 = PixelNormalWS()
  n24 = Saturate(None=Power(Base=Fresnel(ExponentIn=@n22, Normal=@n23) {exponent=1}, Exp=ScalarParameter() {defaultValue=2, parameterName=Fresnel_Power}))
  n25 = CurveAtlasRowParameter(CurveTime=Multiply(A=@n3, B=LinearInterpolate(A=@n21, Alpha=@n24))) {
   atlas=&/CA_DangerRoom_Colors
   curve=&/Curve_DangerRoomHologram_Blue
   parameterName=ColorCurve_01
  }
  NamedRerouteDeclaration(None=@n25) {name=BaseColorandEmissive, nodeColor=<rgba 0 1 0.3906 1>}
  n26 = LocalPosition() {localOrigin=InstancePreSkinning}
  n27 = ShaderStageSwitch(PixelShader=VertexInterpolator(VS=@n26.XYZ).PS, VertexShader=@n26.XYZ)
  n28 = VectorParameter() {parameterName=PositionOffset, defaultValue=<rgba 143 143 0 1>}
  n29 = VectorParameter() {parameterName=GradientLength, defaultValue=<rgba 286 286 286 1>}
  n30 = LinearInterpolate(
   A=ScalarParameter() {parameterName=SpawnIn_GradientOffsetMin}
   B=ScalarParameter() {defaultValue=200, parameterName=SpawnIn_GradientOffsetMax}
   Alpha=NamedRerouteUsage()
  )
  n31 = Saturate(None=Divide(A=Add(A=@n27, B=@n28.RGB), B=Add(A=@n29.RGB, B=@n30)))
  n32 = Subtract(
   A=@n5
   B=LinearInterpolate(
    A=ScalarParameter() {defaultValue=-0.2, parameterName=SpawnIn_NoiseMin}
    B=ScalarParameter() {defaultValue=0.2, parameterName=SpawnIn_NoiseMax}
    Alpha=NamedRerouteUsage()
   )
  )
  n33 = Subtract(A=Add(A=ChannelMaskParameter(None=@n31) {parameterName=GradientDirection}.R), B=Multiply(A=Saturate(None=@n32)) {constB=2})
  NamedRerouteDeclaration(None=Saturate(None=@n33)) {name=Opacity, nodeColor=<rgba 0.7344 1 0 1>}
  n34 = ScalarParameter() {parameterName=OpacityEdgeSamplePos}
  n35 = ScalarParameter() {defaultValue=0.5, parameterName=EdgeRadius}
  n36 = ScalarParameter() {parameterName=EdgeHardness}
  NamedRerouteDeclaration(None=Saturate(None=SphereMask(A=@n33, B=@n34, Radius=@n35, Hardness=@n36))) {name=EdgeMask, nodeColor=<rgba 1 0 0.5859 1>}
  n37 = MaterialFunctionCall(Position=ObjectPositionWS(), UVRadialScale=ScalarParameter() {defaultValue=0.01, parameterName=ScreenUVDistortionNoiseScale}) {materialFunction=&ObjectStabilizedScreenSpaceUVs}
  NamedRerouteDeclaration(None=MaterialFunctionCall(Seed=Time(), Vec2=Ceil(None=@n37.StabilizedUVs)) {materialFunction=&MF_Hash}.Vec2) {name=DistortionNoise, nodeColor=<rgba 1 0 0.9375 1>}
  NamedRerouteUsage()
  NamedRerouteUsage()
  n38 = WorldPosition()
  n39 = ComponentMask(None=MaterialFunctionCall() {materialFunction=&ObjectPivotPoint}.Object Pivot Location) {b=true}
  n40 = Add(A=@n39, B=ScalarParameter() {defaultValue=300, parameterName=PulseHeight}) {constB=1600}
  n41 = ScalarParameter() {defaultValue=1, parameterName=TurnOn}
  n42 = ScalarParameter() {defaultValue=33, parameterName=PulseRadius}
  n43 = Multiply(A=@n42) {constB=0.1}
  n44 = Multiply(
   A=SphereMask(
    A=ComponentMask(None=@n38.XYZ) {b=true}
    B=LinearInterpolate(A=@n39, B=@n40, Alpha=@n41) {constA=-0.25, constB=1.2}
    Radius=LinearInterpolate(A=@n42, B=@n43, Alpha=@n41) {constA=500, constB=100}
   ) {hardnessPercent=0}
   B=OneMinus(None=@n41)
  )
  Reroute(None=@n44)
  ScalarParameter() {parameterName=ColorSampleMin}
  n45 = Sine(None=Multiply(A=Time(), B=Constant() {r=0.4}))
  n46 = LinearInterpolate(
   A=VectorParameter() {parameterName=BobHeightMax, defaultValue=<rgba 0 0 4 0>}.RGB
   B=VectorParameter() {parameterName=BobHeightMin, defaultValue=<rgba 0 0 -4 0>}.RGB
   Alpha=@n45
  )
  n47 = RotateAboutAxis(
   NormalizedRotationAxis=Constant3Vector() {constant=<rgba 0 0 1 0>}
   RotationAngle=LinearInterpolate(A=Constant() {r=-0.01}, B=Constant() {r=0.01}, Alpha=@n45)
   PivotPoint=ObjectPositionWS()
   Position=WorldPosition().XYZ
  )
  n48 = Subtract(A=WorldPosition().XYZ, B=ObjectPositionWS())
  n49 = ObjectRadius()
  n50 = ScalarParameter() {defaultValue=0.025, group=Noise, parameterName=WPOStrengthFromObjectRadius}
  n51 = Multiply(
   A=Multiply(A=Normalize(VectorInput=@n48), B=Constant3Vector() {constant=<rgba 1 1 0 0>})
   B=Add(A=ScalarParameter() {defaultValue=10, group=Noise, parameterName=WPOStrength}, B=Multiply(A=@n49, B=@n50) {constB=0.05})
  )
  n52 = VectorParameter() {group=Noise, parameterName=NoiseTileMult, defaultValue=<rgba 1 1 0.1 1>}
  n53 = Divide(A=ObjectRadius(), B=ScalarParameter() {defaultValue=0.1, group=Noise, parameterName=NoiseSize})
  n54 = WorldPosition()
  n55 = Multiply(A=@n53, B=Multiply(A=Time(), B=VectorParameter() {group=Noise, parameterName=NoisePanSpeed, defaultValue=<rgba -0.08 0.07 -0.2 0>}.RGB))
  n56 = Power(Base=MaterialFunctionCall(
    TextureObject=TextureObject() {samplerType=SAMPLERTYPE_LinearColor, texture=&T_NoisePack_LinearColor}
    TextureSize=Multiply(A=@n52.RGB, B=@n53)
    WorldPosition=Add(A=@n54.XYZ, B=@n55)
   ) {materialFunction=&WorldAlignedTexture}.XY Texture) {constExponent=4}
  n57 = ScalarParameter() {defaultValue=2, group=Noise, parameterName=NoiseMaskMult}
  n58 = ScalarParameter() {group=Noise, parameterName=NoiseAdd}
  n59 = ActorPositionWS()
  n60 = MaterialFunctionCall(Float2=Round(None=Divide(A=@n59) {constB=2000})) {materialFunction=&BreakOutFloat2Components}
  n61 = Add(A=Time(), B=Add(A=@n60.R, B=@n60.R))
  n62 = Multiply(A=ScalarParameter() {defaultValue=0.5, parameterName=FlickerSpeed}, B=@n61)
  n63 = AppendVector(A=AppendVector(A=Sine(None=@n62) {period=1.6}, B=Sine(None=@n62) {period=1.42}), B=Sine(None=@n62) {period=1.51})
  n64 = ScalarParameter() {defaultValue=0.5, group=Noise, parameterName=NoiseFrequency}
  n65 = Add(A=Sine(None=@n62) {period=0.72}, B=Multiply(B=@n64) {constA=0.75}) {constB=0.75}
  n66 = Multiply(A=Add(A=Sine(None=@n62) {period=2.1}, B=Multiply(B=@n64) {constA=0.25}), B=@n65)
  n67 = Multiply(B=@n64) {constA=0.9, constB=0.9}
  n68 = Sine(None=@n62) {period=1.25}
  n69 = Multiply(A=@n65, B=Ceil(None=Add(A=@n67, B=@n68) {constB=0.9}))
  n70 = Sine(None=@n61) {period=0.075}
  n71 = Multiply(A=Saturate(None=Multiply(A=@n66, B=@n69)), B=LinearInterpolate(Alpha=ConstantBiasScale(None=@n70)) {constA=0.7})
  n72 = Multiply(
   A=Multiply(A=Ceil(None=@n63), B=NamedRerouteDeclaration(None=@n71) {name=GlitchTimingMask, nodeColor=<rgba 0.0547 1 0 1>})
   B=ScalarParameter() {defaultValue=2, group=Noise, parameterName=NoiseMult}
  )
  n73 = Saturate(None=Multiply(A=Multiply(A=@n56, B=@n57) {constB=20}, B=Add(A=@n58, B=@n72)))
  n74 = Add(A=Desaturation(None=NamedRerouteDeclaration(None=@n73) {name=Noise, nodeColor=<rgba 1 0.7734 0 1>}), B=Multiply(A=@n44) {constB=2})
  n75 = Subtract(A=ActorPositionWS(), B=WorldPosition().XYZ)
  n76 = NamedRerouteUsage()
  StaticSwitchParameter(
   True=Add(A=Add(A=@n46, B=@n47), B=Multiply(A=@n74, B=@n51) {constB=0})
   False=StaticSwitchParameter(True=Multiply(A=@n75, B=@n76), False=Constant()) {parameterName=UseOpaqueWPO}
  ) {parameterName=UseCoinAnim}
  TextureSampleParameter2D(UVs=NamedRerouteUsage()) {
   parameterName=Normal
   samplerType=SAMPLERTYPE_Normal
   texture=/Hermes/Props/Academy/Academy_GymPlatform_A/Textures/T_Academy_GymPlatform_A_A_N
  }
 ]
}
